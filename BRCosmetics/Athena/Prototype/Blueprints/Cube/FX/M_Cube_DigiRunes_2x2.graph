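# Translucent unlit particle material for Cube "DigiRunes" effects, drawing a glowing rune texture on mesh particles.
# Grid-crack masks, partly sampled by tangent-transformed world position, distort the texture UVs; the result blends dim and bright rune colors, scaled by particle green.
# A stepped 2x2 quadrant mask, rotated in quarter turns and optionally inverted by dynamic parameters, multiplies the texture into an alpha shaped by particle blue and alpha.
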Material /BRCosmetics/Athena/Prototype/Blueprints/Cube/FX/M_Cube_DigiRunes_2x2
graph {
 properties {
  bUsedWithMeshParticles=true
  bUsedWithNiagaraMeshParticles=true
  blendMode=BLEND_Translucent
  pixelDepthOffsetMode=PDOM_Legacy
  shadingModel=MSM_Unlit
  usedShadingModels=Unlit
 }
 nodes [
  n1 = VectorParameter() {parameterName=RuneColorDim}
  n2 = VectorParameter() {parameterName=RuneColorBright, defaultValue=<rgba 1.378 0.1161 2 0>}
  n3 = ComponentMask(None=Transform(None=WorldPosition().XYZ) {transformSourceType=TRANSFORMSOURCE_World, transformType=TRANSFORM_Tangent}) {g=true, r=true}
  n4 = ScalarParameter() {defaultValue=300, parameterName=UVScale}
  n5 = Multiply(
   A=TextureSample(UVs=Divide(A=@n3, B=@n4)) {samplerType=SAMPLERTYPE_LinearColor, texture=/Game/Abilities/Player/Outlander/Actives/GroundWave/FX/T_GridCrack_Mask}.B
   B=TextureSample() {samplerType=SAMPLERTYPE_LinearColor, texture=&T_GridCrack_Mask}.G
  )
  n6 = ScalarParameter() {defaultValue=0.5, parameterName=Distort}
  n7 = TextureSampleParameter2D(UVs=Add(A=TextureCoordinate(), B=Multiply(A=@n5, B=@n6))) {parameterName=Texture, samplerType=SAMPLERTYPE_LinearColor, texture=/Game/Athena/Prototype/Blueprints/Cube/Materials/DFHeightMask}
  n8 = ParticleColor()
  n9 = LinearInterpolate(A=@n7.R, B=@n7.R, Alpha=@n8.R)
  MaterialFunctionCall(Emissive=Multiply(A=ParticleColor().G, B=LinearInterpolate(A=@n1.RGB, B=@n2.RGB, Alpha=@n9))) {materialFunction=&MF_FX_EmissiveAdjustments}
  n10 = DynamicParameter() {paramNames ["Random ", 0-4, na, na]}
  n11 = MaterialFunctionCall(UVs=TextureCoordinate(), Rotation Center=Constant2Vector() {g=0.5, r=0.5}, "Rotation Angle (0-1)"=Divide(A=Ceil(None=@n10.0-4)) {constB=4}) {materialFunction=&CustomRotator}
  n12 = Constant() {r=2}
  n13 = Multiply(A=ComponentMask(None=@n11.Rotated Values) {r=true}, B=@n12)
  n14 = Constant() {r=2}
  n15 = Multiply(A=ComponentMask(None=@n11.Rotated Values) {g=true}, B=@n14)
  n16 = Add(A=Divide(A=Floor(None=@n13), B=@n12), B=Divide(A=Floor(None=@n15), B=@n14))
  n17 = LinearInterpolate(A=OneMinus(None=@n16), B=@n16, Alpha=Saturate(None=@n10."Random "))
  Multiply(A=Power(Base=Multiply(A=@n9, B=@n17), Exp=@n8.B), B=@n8.A)
 ]
}
